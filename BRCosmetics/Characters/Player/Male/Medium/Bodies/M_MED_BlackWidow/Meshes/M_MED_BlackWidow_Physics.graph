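# Physics asset for the male medium BlackWidow player cosmetic body, set up for secondary motion of its collar, arm cloth, shirt back and sword.
# Simple sphere and capsule bodies on head, spine, pelvis, thighs and upper arms form the core, joined by free-motion constraints.
# Accessory bodies hang from that core with tight per-axis angular limits, with collar straps chained to the left collar and the shirt back forming a two-link chain.
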PhysicsAsset /BRCosmetics/Characters/Player/Male/Medium/Bodies/M_MED_BlackWidow/Meshes/M_MED_BlackWidow_Physics
bodies {
 dyn_arm_cloth_l [{shapeType=Sphere, radius=2.7888, center=<5.8094 -0.1769 0.807>}]
 dyn_arm_cloth_r [{shapeType=Sphere, radius=2.7888, center=<-5.8094 0.1769 0.807>}]
 dyn_collar_l [{shapeType=Sphere, radius=1.4252, center=<-0.0063 6.4634 -2.8849>}]
 dyn_collar_r [{shapeType=Sphere, radius=1.4252, center=<-3.5113 -7.2768 2.1492>}]
 dyn_collar_straps [{shapeType=Sphere, radius=2.0918, center=<4.8577 0.1977 -0.2629>}]
 dyn_shirt_back01 [{shapeType=Capsule, length=5.0284, radius=1.7523, center=<3.5099 0.2547 0.1634>, rotation=<rot -4.3356 -89.9997 1.1662>}]
 dyn_shirt_back02 [{shapeType=Capsule, length=5.4779, radius=1.3665, center=<0.8059 1.6178 -0.247>, rotation=<rot 0 111.1133 -178.542>}]
 dyn_sword [{shapeType=Capsule, length=45.6992, radius=1.75, center=<0.4819 -0.0843 -8.9946>, rotation=<rot 0.537 -90 -3.0668>}]
 head [{shapeType=Capsule, length=6.0506, radius=5.328, center=<0.44 0.7796 -1.1273>, rotation=<rot -8.0333 -87.7601 -93.2836>}]
 pelvis [{shapeType=Capsule, length=6.3603, radius=12.3292, center=<-4.2726 -0.4155 0.4616>, rotation=<rot 8.7219 -90.1943 -179.1418>}]
 psd_collar_back [{shapeType=Capsule, length=9.014, radius=2.848, center=<3.4289 -1.9863 0>, rotation=<rot 0 -90 0>}]
 spine_01 [{shapeType=Capsule, length=14.5163, radius=8.2, center=<6.8708 0.8636 -0.0515>, rotation=<rot -0.5999 -89.998 0.5033>}]
 spine_05 [{shapeType=Sphere, radius=2.2256, center=<9.9803 0.5229 0.3938>}]
 thigh_l [{shapeType=Capsule, length=34.0624, radius=7.5959, center=<-18.114 0.559 1.4171>, rotation=<rot -6.7133 88.0081 -88.3283>}]
 thigh_r [{shapeType=Capsule, length=34.0624, radius=7.5959, center=<18.114 -0.559 -1.4171>, rotation=<rot -6.7133 88.0081 91.6716>}]
 thigh_twist_01_r [{shapeType=Capsule, length=8.5, radius=7.4499, center=<-6.1541 -1.2198 2.4316>, rotation=<rot 1.1141 89.9392 -81.2593>}]
 upperarm_twist_01_l [{shapeType=Sphere, radius=3.3555, center=<-0.5535 -1.5032 -0.9229>}]
 upperarm_twist_01_r [{shapeType=Sphere, radius=3.3555, center=<0.5535 1.5032 -0.9229>}]
}
constraints [
 {
  bone1Name=thigh_l
  bone2Name=pelvis
  swing1LimitDegrees=45
  swing1Motion=Free
  swing2LimitDegrees=45
  swing2Motion=Free
  twistLimitDegrees=45
  twistMotion=Free
 }
 {bone1Name=dyn_arm_cloth_l, bone2Name=upperarm_twist_01_l, swing1LimitDegrees=5, swing2LimitDegrees=15, twistLimitDegrees=1}
 {bone1Name=dyn_collar_l, bone2Name=spine_05, swing1LimitDegrees=3, swing2LimitDegrees=7, twistLimitDegrees=5}
 {
  bone1Name=spine_05
  bone2Name=spine_01
  swing1LimitDegrees=45
  swing1Motion=Free
  swing2LimitDegrees=45
  swing2Motion=Free
  twistLimitDegrees=45
  twistMotion=Free
 }
 {
  bone1Name=upperarm_twist_01_l
  bone2Name=spine_05
  swing1LimitDegrees=45
  swing1Motion=Free
  swing2LimitDegrees=45
  swing2Motion=Free
  twistLimitDegrees=45
  twistMotion=Free
 }
 {bone1Name=dyn_collar_straps, bone2Name=dyn_collar_l, swing1LimitDegrees=3, swing2LimitDegrees=15, twistLimitDegrees=1}
 {bone1Name=dyn_collar_r, bone2Name=spine_05, swing1LimitDegrees=3, swing2LimitDegrees=30, twistLimitDegrees=5}
 {
  bone1Name=upperarm_twist_01_r
  bone2Name=spine_05
  swing1LimitDegrees=45
  swing1Motion=Free
  swing2LimitDegrees=45
  swing2Motion=Free
  twistLimitDegrees=45
  twistMotion=Free
 }
 {bone1Name=dyn_arm_cloth_r, bone2Name=upperarm_twist_01_r, swing1LimitDegrees=5, swing2LimitDegrees=15, twistLimitDegrees=1}
 {
  bone1Name=head
  bone2Name=spine_05
  swing1LimitDegrees=45
  swing1Motion=Free
  swing2LimitDegrees=45
  swing2Motion=Free
  twistLimitDegrees=45
  twistMotion=Free
 }
 {bone1Name=psd_collar_back, bone2Name=spine_05, swing1LimitDegrees=15, swing2LimitDegrees=3, twistLimitDegrees=1}
 {bone1Name=dyn_sword, bone2Name=pelvis, swing1LimitDegrees=2, swing2LimitDegrees=20, twistLimitDegrees=25}
 {
  bone1Name=spine_01
  bone2Name=pelvis
  swing1LimitDegrees=45
  swing1Motion=Free
  swing2LimitDegrees=45
  swing2Motion=Free
  twistLimitDegrees=45
  twistMotion=Free
 }
 {bone1Name=dyn_shirt_back01, bone2Name=spine_01, swing1LimitDegrees=15, swing2LimitDegrees=30, twistLimitDegrees=1}
 {bone1Name=dyn_shirt_back02, bone2Name=dyn_shirt_back01, swing1LimitDegrees=20, swing2LimitDegrees=15, twistLimitDegrees=1}
 {
  bone1Name=thigh_r
  bone2Name=pelvis
  swing1LimitDegrees=45
  swing1Motion=Free
  swing2LimitDegrees=45
  swing2Motion=Free
  twistLimitDegrees=45
  twistMotion=Free
 }
 {
  bone1Name=thigh_twist_01_r
  bone2Name=thigh_r
  swing1LimitDegrees=45
  swing1Motion=Free
  swing2LimitDegrees=45
  swing2Motion=Free
  twistLimitDegrees=45
  twistMotion=Free
 }
]
