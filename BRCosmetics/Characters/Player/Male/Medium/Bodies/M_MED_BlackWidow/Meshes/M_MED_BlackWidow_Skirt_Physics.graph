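# Physics asset for the skirt of the M_MED_BlackWidow male character body, defining collision shapes and joints for its dynamic skirt bones.
# Capsules cover the pelvis, thighs and calves, while small base spheres link the front and back skirt chains to the thigh twist bones with free rotation.
# Front and right-back panels chain with tighter angular settings.
# Back-left lacks its middle body, so its lower capsule joins the base directly with free rotation.
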PhysicsAsset /BRCosmetics/Characters/Player/Male/Medium/Bodies/M_MED_BlackWidow/Meshes/M_MED_BlackWidow_Skirt_Physics
bodies {
 calf_l [{shapeType=Capsule, length=14.5, radius=7.5, center=<-9.0684 -5.0351 -0.0004>, rotation=<rot 90 0 -8.2196>}]
 calf_r [{shapeType=Capsule, length=15.75, radius=7.75, center=<13.6925 3.0508 0.0004>, rotation=<rot 90 0 163.667>}]
 calf_twist_01_l [{shapeType=Capsule, length=22.25, radius=8.25, center=<1.8274 -1.1856 -0.0001>, rotation=<rot 90 90 100.625>}]
 calf_twist_01_r [{shapeType=Capsule, length=18, radius=8, center=<0.8324 0.6868 0.0002>, rotation=<rot 90 0 -169.3751>}]
 dyn_skirt_back_a_base_l [{shapeType=Sphere, radius=1.255}]
 dyn_skirt_back_a_base_r [{shapeType=Sphere, radius=1.255}]
 dyn_skirt_back_a_r [{shapeType=Box, extentX=13, extentY=4.5, extentZ=13, center=<-5.925 0.0186 -0.0001>, rotation=<rot 0 -5.5471 -180>}]
 dyn_skirt_back_b_l [{shapeType=Capsule, length=12.6456, radius=7.2305, center=<7.9183 -4.7201 0.0001>, rotation=<rot 0 -93.7142 -90.0001>}]
 dyn_skirt_back_b_r [{shapeType=Capsule, length=12.6456, radius=7.2305, center=<-7.7791 4.0696 -0.0001>, rotation=<rot 0 -93.7142 89.9999>}]
 dyn_skirt_front_a_base_l [{shapeType=Sphere, radius=1.255}]
 dyn_skirt_front_a_base_r [{shapeType=Sphere, radius=1.255}]
 dyn_skirt_front_a_l [{shapeType=Box, extentX=14, extentY=3, extentZ=18, center=<3.9628 -0.0001 -0.1748>}]
 dyn_skirt_front_a_r [{shapeType=Box, extentX=14, extentY=4, extentZ=18, center=<-3.9778 0.0001 -0.0001>, rotation=<rot 0 0 -180>}]
 dyn_skirt_front_b_l [{shapeType=Box, extentX=20, extentY=4, extentZ=18.5, center=<8.671 0 -0.3838>}]
 dyn_skirt_front_b_r [{shapeType=Box, extentX=20, extentY=5, extentZ=19, center=<-8.7037 0 0>, rotation=<rot 0 0 -180>}]
 pelvis [{shapeType=Capsule, length=8, radius=10.75}]
 thigh_twist_01_l [{shapeType=Capsule, length=34.0182, radius=8.1822, rotation=<rot -0.0723 -89.3201 -90.8948>}]
 thigh_twist_01_r [{shapeType=Capsule, length=34.5182, radius=7.6822, rotation=<rot -0.0792 -89.7628 89.1049>}]
}
constraints [
 {
  bone1Name=thigh_twist_01_l
  bone2Name=pelvis
  swing1LimitDegrees=45
  swing1Motion=Free
  swing2LimitDegrees=45
  swing2Motion=Free
  twistLimitDegrees=45
  twistMotion=Free
 }
 {
  bone1Name=thigh_twist_01_r
  bone2Name=pelvis
  swing1LimitDegrees=45
  swing1Motion=Free
  swing2LimitDegrees=45
  swing2Motion=Free
  twistLimitDegrees=45
  twistMotion=Free
 }
 {
  bone1Name=dyn_skirt_front_a_base_r
  bone2Name=thigh_twist_01_r
  swing1LimitDegrees=45
  swing1Motion=Free
  swing2LimitDegrees=45
  swing2Motion=Free
  twistLimitDegrees=45
  twistMotion=Free
 }
 {
  bone1Name=dyn_skirt_back_a_base_r
  bone2Name=thigh_twist_01_r
  swing1LimitDegrees=45
  swing1Motion=Free
  swing2LimitDegrees=45
  swing2Motion=Free
  twistLimitDegrees=45
  twistMotion=Free
 }
 {
  bone1Name=dyn_skirt_front_a_base_l
  bone2Name=thigh_twist_01_l
  swing1LimitDegrees=45
  swing1Motion=Free
  swing2LimitDegrees=45
  swing2Motion=Free
  twistLimitDegrees=45
  twistMotion=Free
 }
 {
  bone1Name=dyn_skirt_back_a_base_l
  bone2Name=thigh_twist_01_l
  swing1LimitDegrees=45
  swing1Motion=Free
  swing2LimitDegrees=45
  swing2Motion=Free
  twistLimitDegrees=45
  twistMotion=Free
 }
 {bone1Name=dyn_skirt_front_a_l, bone2Name=dyn_skirt_front_a_base_l, swing1LimitDegrees=45, swing2LimitDegrees=15, twistLimitDegrees=5}
 {bone1Name=dyn_skirt_front_b_l, bone2Name=dyn_skirt_front_a_l, swing1LimitDegrees=45, swing2LimitDegrees=15, twistLimitDegrees=5}
 {bone1Name=dyn_skirt_front_a_r, bone2Name=dyn_skirt_front_a_base_r, swing1LimitDegrees=45, swing2LimitDegrees=15, twistLimitDegrees=5}
 {bone1Name=dyn_skirt_front_b_r, bone2Name=dyn_skirt_front_a_r, swing1LimitDegrees=45, swing2LimitDegrees=15, twistLimitDegrees=5}
 {bone1Name=dyn_skirt_back_a_r, bone2Name=dyn_skirt_back_a_base_r, swing1LimitDegrees=45, swing2LimitDegrees=15, twistLimitDegrees=5}
 {bone1Name=dyn_skirt_back_b_r, bone2Name=dyn_skirt_back_a_r, swing1LimitDegrees=45, swing2LimitDegrees=15, twistLimitDegrees=5}
 {
  bone1Name=calf_l
  bone2Name=pelvis
  swing1LimitDegrees=45
  swing1Motion=Free
  swing2LimitDegrees=45
  swing2Motion=Free
  twistLimitDegrees=45
  twistMotion=Free
 }
 {
  bone1Name=calf_r
  bone2Name=pelvis
  swing1LimitDegrees=45
  swing1Motion=Free
  swing2LimitDegrees=45
  swing2Motion=Free
  twistLimitDegrees=45
  twistMotion=Free
 }
 {
  bone1Name=calf_twist_01_l
  bone2Name=calf_l
  swing1LimitDegrees=45
  swing1Motion=Free
  swing2LimitDegrees=45
  swing2Motion=Free
  twistLimitDegrees=45
  twistMotion=Free
 }
 {
  bone1Name=calf_twist_01_r
  bone2Name=calf_r
  swing1LimitDegrees=45
  swing1Motion=Free
  swing2LimitDegrees=45
  swing2Motion=Free
  twistLimitDegrees=45
  twistMotion=Free
 }
 {
  bone1Name=dyn_skirt_back_b_l
  bone2Name=dyn_skirt_back_a_base_l
  swing1LimitDegrees=45
  swing1Motion=Free
  swing2LimitDegrees=45
  swing2Motion=Free
  twistLimitDegrees=45
  twistMotion=Free
 }
]
