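# Character-part AnimBlueprint for the Black Widow player body that copies the attached parent's pose and adds secondary motion to sword and skirt.
# Sword physics switches off during tactical sprint, knee slide and jump/fall (skydiving excepted), with additive offsets covering slide and jump/fall.
# A root-space RigidBody simulates the skirt, then additive bone corrections adjust its back.
# The EventGraph drives those corrections from thigh rotations, stationary crouching and sideways crouch movement.
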AnimBlueprint /BRCosmetics/Characters/Player/Male/Medium/Bodies/M_MED_BlackWidow/Meshes/M_MED_BlackWidow_Skeleton_ABP
target_skeleton=&/M_MED_BlackWidow_Skeleton
graphs [
 AnimGraph {nodes [
   n1 = RigidBody(
    ComponentPose=Local To Component(LocalPose=Copy Pose From Mesh() {bUseAttachedParent=true})
    Alpha=Get Sword RigidBody Alpha() -> (Sword RigidBody Alpha:float)
   ) {
    bClampLinearTranslationLimitToRefPose=true
    bOverrideWorldGravity=true
    overridePhysicsAsset=&/M_MED_BlackWidow_Physics
    alphaBoolBlend {blendOption=Linear}
    overrideWorldGravity {z=-980}
   }
   n2 = "Transform (Modify) Bone - Bone: dyn_sword"(
    ComponentPose=@n1
    Alpha=Get Sword ModifyTransform KneeSlide Alpha() -> (Sword ModifyTransform KneeSlide Alpha:float)
    Translation="-1,-3,0"
    Rotation="-10,30,20"
    Scale="1,1,1"
   ) {
    rotationMode=BMM_Additive
    rotationSpace=BCS_BoneSpace
    translationMode=BMM_Additive
    translationSpace=BCS_BoneSpace
    alphaBoolBlend {blendOption=Linear}
    alphaScaleBiasClamp {bInterpResult=true}
    boneToModify {boneName=dyn_sword}
    rotation=<rot -10 30 20>
    translation=<-1 -3>
   }
   n3 = "Transform (Modify) Bone - Bone: dyn_sword"(
    ComponentPose=@n2
    Alpha=Get Sword ModifyTransform JumpFall Alpha() -> (Sword ModifyTransform JumpFall Alpha:float)
    Translation="0,-3,0"
    Rotation="0,30,20"
    Scale="1,1,1"
   ) {
    rotationMode=BMM_Additive
    rotationSpace=BCS_BoneSpace
    translationMode=BMM_Additive
    translationSpace=BCS_BoneSpace
    alphaBoolBlend {blendOption=Linear}
    alphaScaleBiasClamp {bInterpResult=true}
    boneToModify {boneName=dyn_sword}
    rotation {roll=20, yaw=30}
    translation {y=-3}
   }
   n4 = RigidBody(ComponentPose=@n3) {
    bClampLinearTranslationLimitToRefPose=true
    bOverrideWorldGravity=true
    overridePhysicsAsset=&/M_MED_BlackWidow_Skirt_Physics
    simulationSpace=BaseBoneSpace
    alphaBoolBlend {blendOption=Linear}
    baseBoneRef {boneName=root}
    componentAppliedLinearAccClamp {y=4000}
    componentLinearAccScale {y=0.3}
    overrideWorldGravity {z=-1300}
   }
   n5 = "Transform (Modify) Bone - Bone: dyn_skirt_back_a_base_l"(ComponentPose=@n4, Alpha=Get legUp_L() -> (legUp_L:double), Translation="-5,0,0", Rotation="0,-80,0", Scale="1,1,1") {
    rotationMode=BMM_Additive
    rotationSpace=BCS_BoneSpace
    translationMode=BMM_Additive
    translationSpace=BCS_BoneSpace
    alphaBoolBlend {blendOption=Linear}
    alphaScaleBiasClamp {bMapRange=true, inRange {max=0, min=-110}, outRange {max=0, min=1}}
    boneToModify {boneName=dyn_skirt_back_a_base_l}
    rotation {yaw=-80}
    translation {x=-5}
   }
   n6 = "Transform (Modify) Bone - Bone: dyn_skirt_back_a_base_r"(ComponentPose=@n5, Alpha=Get legUp_R() -> (legUp_R:double), Translation="-5,0,0", Rotation="0,-80,0", Scale="1,1,1") {
    rotationMode=BMM_Additive
    rotationSpace=BCS_BoneSpace
    translationMode=BMM_Additive
    translationSpace=BCS_BoneSpace
    alphaBoolBlend {blendOption=Linear}
    alphaScaleBiasClamp {bMapRange=true, inRange {max=0, min=-110}, outRange {max=0, min=1}}
    boneToModify {boneName=dyn_skirt_back_a_base_r}
    rotation {yaw=-80}
    translation {x=-5}
   }
   n7 = "Transform (Modify) Bone - Bone: dyn_skirt_back_a_base_r"(ComponentPose=@n6, Alpha=Get crouching() -> (crouching:double), Translation="0,2,12", Rotation="20,0,-25", Scale="1,1,1") {
    rotationMode=BMM_Additive
    rotationSpace=BCS_BoneSpace
    translationMode=BMM_Additive
    translationSpace=BCS_BoneSpace
    alphaBoolBlend {blendOption=Linear}
    alphaScaleBiasClamp {inRange {max=0, min=-110}, outRange {max=0, min=1}}
    boneToModify {boneName=dyn_skirt_back_a_base_r}
    rotation {pitch=20, roll=-25}
    translation {y=2, z=12}
   }
   n8 = Get crouchMoveRightLegL() -> (crouchMoveRightLegL:double)
   n9 = "Transform (Modify) Bone - Bone: dyn_skirt_back_a_base_l"(ComponentPose=@n7, Alpha=@n8, Translation="0,0,5", Rotation="100,0,0", Scale="1,1,1") {
    rotationMode=BMM_Additive
    rotationSpace=BCS_BoneSpace
    translationMode=BMM_Additive
    translationSpace=BCS_BoneSpace
    alphaBoolBlend {blendOption=Linear}
    alphaScaleBiasClamp {inRange {max=0, min=-110}, outRange {max=0, min=1}}
    boneToModify {boneName=dyn_skirt_back_a_base_l}
    rotation {pitch=100}
    translation {z=5}
   }
   n10 = "Transform (Modify) Bone - Bone: dyn_skirt_back_a_base_r"(ComponentPose=@n9, Alpha=@n8, Translation="0,0,0", Rotation="50,0,0", Scale="1,1,1") {
    rotationMode=BMM_Additive
    rotationSpace=BCS_BoneSpace
    translationMode=BMM_Additive
    translationSpace=BCS_BoneSpace
    alphaBoolBlend {blendOption=Linear}
    alphaScaleBiasClamp {inRange {max=0, min=-110}, outRange {max=0, min=1}}
    boneToModify {boneName=dyn_skirt_back_a_base_r}
    rotation {pitch=50}
   }
   n11 = Get crouchMoveLeftLegR() -> (crouchMoveLeftLegR:double)
   n12 = "Transform (Modify) Bone - Bone: dyn_skirt_back_a_base_r"(ComponentPose=@n10, Alpha=@n11, Translation="0,0,-5", Rotation="100,0,0", Scale="1,1,1") {
    rotationMode=BMM_Additive
    rotationSpace=BCS_BoneSpace
    translationMode=BMM_Additive
    translationSpace=BCS_BoneSpace
    alphaBoolBlend {blendOption=Linear}
    alphaScaleBiasClamp {inRange {max=0, min=-110}, outRange {max=0, min=1}}
    boneToModify {boneName=dyn_skirt_back_a_base_r}
    rotation {pitch=100}
    translation {z=-5}
   }
   n13 = "Transform (Modify) Bone - Bone: dyn_skirt_back_a_base_l"(ComponentPose=@n12, Alpha=@n11, Translation="0,0,0", Rotation="50,0,0", Scale="1,1,1") {
    rotationMode=BMM_Additive
    rotationSpace=BCS_BoneSpace
    translationMode=BMM_Additive
    translationSpace=BCS_BoneSpace
    alphaBoolBlend {blendOption=Linear}
    alphaScaleBiasClamp {inRange {max=0, min=-110}, outRange {max=0, min=1}}
    boneToModify {boneName=dyn_skirt_back_a_base_l}
    rotation {pitch=50}
   }
   Output Pose(Result=Component To Local(ComponentPose=@n13)) {layerGroup=None}
  ]}
 EventGraph {nodes [
   Event BlueprintUpdateAnimation()
   > n1 = Sequence!(execute) -> (then_0, then_1, then_2:Exec, then_3:Exec, then_4:Exec, then_5:Exec, then_6:Exec)
   n2 = GetOwnerSkeletalMeshComponent()
   n3 = GetDeltaTransformFromRefPose(self=@n2, BoneName=thigh_l, BaseName=pelvis)
   > Set legUp_L(legUp_L:double=BreakRotator(InRot=BreakTransform(InTransform=@n3).Rotation).Yaw) -> (Output_Get:double)
   n4 = GetDeltaTransformFromRefPose(self=@n2, BoneName=thigh_r, BaseName=pelvis)
   Set legUp_R(execute=@n1.then_1, legUp_R:double=BreakRotator(InRot=BreakTransform(InTransform=@n4).Rotation).Yaw) -> (Output_Get:double)
   n5 = "Not Equal (Boolean)"!(A:bool=Get bIsMoving2D(), B:bool=true) -> (ReturnValue:bool) {is=title K2Node_CallFunction}
   n6 = Get bIsCrouching()
   Set crouching(execute=@n1.then_2, crouching:double="To Float (Boolean)"(InBool=AND Boolean(A=@n6, B=@n5))) -> (Output_Get:double)
   n7 = Sequence!(execute=@n1.then_3) -> (then_0, then_1, then_2)
   n8 = GetOwnerDirection!(self:Custom Character Part Anim Instance Object Reference) -> (ReturnValue:float)
   > Set IsMovingL(IsMovingL:bool=AND Boolean(
     A="float > float"(A=@n8, B=-120) {is=title K2Node_CallFunction}
     B="float < float"!(A:double=@n8, B:double=-60) -> (ReturnValue:bool) {is=title K2Node_CallFunction}
    )) -> (Output_Get:bool)
   n9 = Get bIsCrouching()
   > n10 = Branch(Condition=AND Boolean(A=Get IsMovingL() -> (IsMovingL:bool), B=@n9))
   n11 = BreakTransform(InTransform=GetDeltaTransformFromRefPose(self=@n2, BoneName=thigh_r, BaseName=pelvis))
   > Set crouchMoveLeftLegR(crouchMoveLeftLegR:double=MapRangeClamped(Value=BreakRotator(InRot=@n11.Rotation).Pitch, InRangeA=-15, InRangeB=0, OutRangeA=1, OutRangeB=0)) -> (Output_Get:double)
   Set crouchMoveLeftLegR(execute=@n10.else, crouchMoveLeftLegR=0)
   Set IsMovingR(execute=@n7.then_1, IsMovingR:bool=AND Boolean(A="float > float"(A=@n8, B=60), B="float < float"(A=@n8, B=120))) -> (Output_Get:bool)
   > n12 = Branch(Condition=AND Boolean(A=@n9, B=Get IsMovingR() -> (IsMovingR:bool)))
   n13 = BreakTransform(InTransform=GetDeltaTransformFromRefPose(self=@n2, BoneName=thigh_l, BaseName=pelvis))
   > Set crouchMoveRightLegL(crouchMoveRightLegL:double=MapRangeClamped(Value=BreakRotator(InRot=@n13.Rotation).Pitch, InRangeA=-15, InRangeB=0, OutRangeA=1, OutRangeB=0)) -> (Output_Get:double)
   Set crouchMoveRightLegL(execute=@n12.else, crouchMoveRightLegL=0)
   n14 = NOT Boolean(A=Get bIsKneeSliding())
   n15 = NOT Boolean(A=Get bIsTacticalSprinting())
   n16 = NOT Boolean(A=Get bIsSkyDiving())
   n17 = Get bIsFalling()
   n18 = NOT Boolean(A=OR Boolean(A=Get bIsJumping(), B=AND Boolean(A=@n17, B=@n16)))
   Set Sword RigidBody Alpha(
    execute=@n1.then_4
    Sword RigidBody Alpha:float=float * float!(A=1, B="To Float (Boolean)"(InBool=@n15), C:double="To Float (Boolean)"(InBool=@n14), D:double="To Float (Boolean)"(InBool=@n18)) -> (ReturnValue) {is=title K2Node_PromotableOperator}
   ) -> (Output_Get:float)
   Set Sword ModifyTransform KneeSlide Alpha(execute=@n1.then_5, Sword ModifyTransform KneeSlide Alpha:float="To Float (Boolean)"(InBool=Get bIsKneeSliding())) -> (Output_Get:float)
   n19 = AND Boolean(A=Get bIsFalling(), B=NOT Boolean(A=Get bIsSkyDiving()))
   n20 = Get bIsJumping()
   Set Sword ModifyTransform JumpFall Alpha(execute=@n1.then_6, Sword ModifyTransform JumpFall Alpha:float="To Float (Boolean)"(InBool=OR Boolean(A=@n20, B=@n19))) -> (Output_Get:float)
   TryGetPawnOwner()
  ]}
]
variables [
 crouching {type=number}
 crouchMoveLeftLegR {type=number}
 crouchMoveRightLegL {type=number}
 IsMovingL {type=boolean}
 IsMovingR {type=boolean}
 legUp_L {type=number}
 legUp_R {type=number}
 Sword ModifyTransform JumpFall Alpha {type=number}
 Sword ModifyTransform KneeSlide Alpha {type=number}
 Sword RigidBody Alpha {type=number}
]
